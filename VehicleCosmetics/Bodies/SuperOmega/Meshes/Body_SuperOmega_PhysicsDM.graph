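# Physics asset for the SuperOmega vehicle body cosmetic mesh, providing its simplified collision shape.
# It holds a single body on the root bone, made of one box offset forward and upward to approximate the car's volume.
# There are no constraints, so the body simulates as one rigid piece.
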PhysicsAsset /VehicleCosmetics/Bodies/SuperOmega/Meshes/Body_SuperOmega_PhysicsDM
bodies {root [{shapeType=Box, extentX=444, extentY=216, extentZ=116, shapeName=root_box, center=<33.775 0 96.2586>}]}
constraints []
